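; Bulkhead door prop Blueprint for the SharkBait hull set in the Helios environment, with no door behavior in its own graphs.
; Begin play, overlap and tick events are empty, and the construction script only calls its parent.
; Exposed Asset Provider overrides extend the parent by adding version 1 and registering the Material variable as an allowed indirect asset reference.
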(event EventBeginPlay)

(event Collision|EventActorBeginOverlap
  (params (OtherActor :type Actor%20Object%20Reference)))

(event EventTick
  (params (DeltaSeconds :type Float%20%28single-precision%29)))

(fn BuildExposedAssetProviderVersion(Implementation)
  (params (Version :type Fort%20Exposed%20Asset%20Provider%20Version%20Structure%20%28by%20ref%29))
  (|CallParentFunction Version)
  (ExposedAssetProvider|AddVersion Version 1)
  (return))

(fn CollectAssetReferencestoAllow(Implementation)
  (params (Collector :type Fort%20Exposed%20Asset%20Reference%20Collector%20Structure%20%28by%20ref%29))
  (|CallParentFunction Collector)
  (ExposedAssetProvider|CollectIndirectAssetReference Collector (LocalVariables|GetMaterial))
  (return))

(fn ConstructionScript
  (|CallParentFunction))
